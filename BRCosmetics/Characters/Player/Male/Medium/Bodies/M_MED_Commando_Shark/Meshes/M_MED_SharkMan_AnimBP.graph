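# Animation Blueprint for the SharkMan commando body, adding secondary motion to a pose copied from the owner's skeletal mesh.
# Chained Anim Dynamics nodes spring the calf fins, hand fins and sash bones in root-relative space.
# Each update measures spine_05 against spine_01 from ref pose, driving two additive corrective transforms on the lower sash for spine twist and bend.
# A speed-scaled right-pouch dynamics node is not wired into the output.
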AnimBlueprint /BRCosmetics/Characters/Player/Male/Medium/Bodies/M_MED_Commando_Shark/Meshes/M_MED_SharkMan_AnimBP
target_skeleton=&/M_MED_commando_SharkMan_Skeleton
graphs [
 AnimGraph {nodes [
   Anim Dynamics - Bone: dyn_pouch_r(Alpha=MapRangeClamped(Value=Get PawnSpeed2D(), InRangeA=0, InRangeB=100, OutRangeA=0.1), ExternalForce="0,0,0") {
    bGravityOverrideInSimSpace=true
    numSolverIterationsPostUpdate=8
    numSolverIterationsPreUpdate=32
    simulationSpace=RootRelative
    alphaBoolBlend {blendOption=Linear}
    boundBone {boneName=dyn_pouch_r}
    componentLinearAccScale=<1 1 1>
    componentLinearVelScale=<1 1 1>
    physicsBodyDefinitions [{
      sphereCollisionRadius=0
      boundBone {boneName=dyn_pouch_r}
      boxExtents=<20 20 15>
      constraintSetup {angularLimitsMax=<0 0 0>, angularLimitsMin=<0 0 0>, angularTarget=<1 0 0>, linearAxesMax=<0.4 0.2 0.1>, linearAxesMin=<-0.4 -0.2 -0.1>}
      localJointOffset=<0 0 0>
     }]
   }
   n1 = Get MeshToCopy() -> (MeshToCopy:Skeletal Mesh Component Object Reference)
   n2 = Anim Dynamics - Bone: calf_fin_l(ComponentPose=Local To Component(LocalPose=Copy Pose From Mesh(SourceMeshComponent=@n1)), Alpha=1, ExternalForce="0,0,0") {
    alpha=0.9
    angularDampingOverride=0.98
    angularSpringConstant=1500
    bAngularSpring=true
    bGravityOverrideInSimSpace=true
    bOverrideAngularDamping=true
    numSolverIterationsPostUpdate=4
    numSolverIterationsPreUpdate=16
    simulationSpace=RootRelative
    alphaBoolBlend {blendOption=Linear}
    boundBone {boneName=calf_fin_l}
    physicsBodyDefinitions [{
      sphereCollisionRadius=0
      boundBone {boneName=calf_fin_l}
      boxExtents=<25 10 10>
      constraintSetup {angularTargetAxis=AxisZ, twistAxis=AxisY, angularLimitsMax=<30 0 0>, angularLimitsMin=<-30 0 0>, angularTarget=<0 0 1>}
      localJointOffset=<0 0 0>
     }]
   }
   n3 = Anim Dynamics - Bone: calf_fin_r(ComponentPose=@n2, Alpha=1, ExternalForce="0,0,0") {
    alpha=0.9
    angularDampingOverride=0.98
    angularSpringConstant=1500
    bAngularSpring=true
    bGravityOverrideInSimSpace=true
    bOverrideAngularDamping=true
    numSolverIterationsPostUpdate=4
    numSolverIterationsPreUpdate=16
    simulationSpace=RootRelative
    alphaBoolBlend {blendOption=Linear}
    boundBone {boneName=calf_fin_r}
    physicsBodyDefinitions [{
      sphereCollisionRadius=0
      boundBone {boneName=calf_fin_r}
      boxExtents=<25 10 10>
      constraintSetup {angularTargetAxis=AxisZ, twistAxis=AxisY, angularLimitsMax=<30 0 0>, angularLimitsMin=<-30 0 0>, angularTarget=<0 0 1>}
      localJointOffset=<0 0 0>
     }]
   }
   n4 = Anim Dynamics - Bone: hand_fin_l(ComponentPose=@n3, Alpha=1, ExternalForce="0,0,0") {
    angularDampingOverride=0.9
    angularSpringConstant=500
    bAngularSpring=true
    bGravityOverrideInSimSpace=true
    bOverrideAngularDamping=true
    simulationSpace=RootRelative
    alphaBoolBlend {blendOption=Linear}
    boundBone {boneName=hand_fin_l}
    physicsBodyDefinitions [{
      sphereCollisionRadius=0
      boundBone {boneName=hand_fin_l}
      boxExtents=<5 10 10>
      constraintSetup {angularLimitsMax=<7 1 10>, angularLimitsMin=<-7 -1 -5>, angularTarget=<1 0 0>}
      localJointOffset=<0 0 0>
     }]
   }
   n5 = Anim Dynamics - Bone: hand_fin_r(ComponentPose=@n4, Alpha=1, ExternalForce="0,0,0") {
    angularDampingOverride=0.9
    angularSpringConstant=500
    bAngularSpring=true
    bGravityOverrideInSimSpace=true
    bOverrideAngularDamping=true
    simulationSpace=RootRelative
    alphaBoolBlend {blendOption=Linear}
    boundBone {boneName=hand_fin_r}
    physicsBodyDefinitions [{
      sphereCollisionRadius=0
      boundBone {boneName=hand_fin_r}
      boxExtents=<5 10 10>
      constraintSetup {angularLimitsMax=<7 1 5>, angularLimitsMin=<-7 -1 -10>, angularTarget=<1 0 0>}
      localJointOffset=<0 0 0>
     }]
   }
   n6 = Anim Dynamics - Bone: sash_top(ComponentPose=@n5, Alpha=1, ExternalForce="0,0,0") {
    angularSpringConstant=1600
    bGravityOverrideInSimSpace=true
    bLinearSpring=true
    simulationSpace=RootRelative
    alphaBoolBlend {blendOption=Linear}
    boundBone {boneName=sash_top}
    physicsBodyDefinitions [{
      sphereCollisionRadius=0
      boundBone {boneName=sash_top}
      boxExtents=<15 15 15>
      constraintSetup {angularLimitsMax=<0 0 0>, angularLimitsMin=<0 0 0>, angularTarget=<1 0 0>, linearAxesMax=<2 1 1>, linearAxesMin=<0 -1 -1>}
      localJointOffset=<0 0 0>
     }]
   }
   n7 = Anim Dynamics - Bone: sash_btm(ComponentPose=@n6, Alpha=1, ExternalForce="0,0,0") {
    angularDampingOverride=0.99
    angularSpringConstant=1600
    bGravityOverrideInSimSpace=true
    bLinearSpring=true
    simulationSpace=RootRelative
    alphaBoolBlend {blendOption=Linear}
    boundBone {boneName=sash_btm}
    componentAppliedLinearAccClamp=<3000 3000 3000>
    componentLinearAccScale=<0.5 0.5 0.5>
    componentLinearVelScale=<0.5 0.5 0.5>
    physicsBodyDefinitions [{
      sphereCollisionRadius=0
      boundBone {boneName=sash_btm}
      boxExtents=<15 15 15>
      constraintSetup {angularLimitsMax=<0 0 0>, angularLimitsMin=<0 0 0>, angularTarget=<1 0 0>, linearAxesMax=<2 0 1>}
      localJointOffset=<0 0 0>
     }]
   }
   n8 = "Transform (Modify) Bone - Bone: sash_btm"(
    ComponentPose=@n7
    Alpha=Get spine_twist() -> (spine_twist:double)
    Translation="2,-1.440,-0.204"
    Rotation="-6,0.000,0.000"
    Scale="1.000,1.250,1.000"
   ) {
    rotationMode=BMM_Additive
    rotationSpace=BCS_BoneSpace
    scaleMode=BMM_Additive
    scaleSpace=BCS_BoneSpace
    translationMode=BMM_Additive
    translationSpace=BCS_BoneSpace
    alphaBoolBlend {blendOption=Linear}
    boneToModify {boneName=sash_btm}
    rotation {pitch=-6}
    scale {y=1.25}
    translation=<2 -1.44 -0.204>
   }
   n9 = "Transform (Modify) Bone - Bone: sash_btm"(ComponentPose=@n8, Alpha=Get spine_rot_bk() -> (spine_rot_bk:double), Translation="2,-2,0", Rotation="-15,5,0", Scale="1,1,1") {
    rotationMode=BMM_Additive
    rotationSpace=BCS_BoneSpace
    scaleMode=BMM_Additive
    scaleSpace=BCS_BoneSpace
    translationMode=BMM_Additive
    translationSpace=BCS_BoneSpace
    alphaBoolBlend {blendOption=Linear}
    boneToModify {boneName=sash_btm}
    rotation {pitch=-15, yaw=5}
    translation=<2 -2>
   }
   Output Pose(Result=Component To Local(ComponentPose=@n9)) {layerGroup=None}
  ]}
 EventGraph {nodes [
   Event BlueprintUpdateAnimation()
   > Set MeshToCopy(MeshToCopy:Skeletal Mesh Component Object Reference=GetOwnerSkeletalMeshComponent()) -> (Output_Get:Skeletal Mesh Component Object Reference)
   n1 = GetOwnerSkeletalMeshComponent()
   n2 = BreakRotator(InRot=BreakTransform(InTransform=GetDeltaTransformFromRefPose(self=@n1, BoneName=spine_05, BaseName=spine_01)).Rotation)
   > Set spine_twist(spine_twist:double=MapRangeClamped(Value=@n2, InRangeA=0, InRangeB=50)) -> (Output_Get:double)
   > Set spine_rot_bk(spine_rot_bk:double=MapRangeClamped(Value=@n2.Yaw, InRangeA=0, InRangeB=-50)) -> (Output_Get:double)
   TryGetPawnOwner()
  ]}
]
variables [
 MeshToCopy {type=SkeletalMeshComponent}
 SkelMesh {type=SkeletalMeshComponent}
 spine_rot_bk {type=number}
 spine_twist {type=number}
]
